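# Post-process Animation Blueprint for the GroundGhost biped cosmetic companion, adjusting belly bones on the incoming pose.
# In component space, two chained Constraint nodes tie the translation of belly_l to foot_l and belly_r to foot_r at full weight.
# A Pose Asset blend node reads a variable defaulting to the companion template but is not wired into the output, and the Event Graph is empty.
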AnimBlueprint /CosmeticCompanions/Assets/Biped/GroundGhost/Meshes/GroundGhost_Biped_Post_AnimBP
target_skeleton=&/GroundGhost_Biped_Skeleton
graphs [
 AnimGraph {nodes [
   Pose!(PoseAsset:Pose Asset Object Reference=Get PoseAsset() -> (PoseAsset:Pose Asset Object Reference), SourcePose:Pose Link Structure) -> (Pose:Pose Link Structure) {is=title AnimGraphNode_PoseBlendNode, bIgnorePoseAssetSourceAnimationCurves=true, blendOption=Linear}
   n1 = "Constraint - belly_l (T)"(ComponentPose=Local To Component(LocalPose=Input Pose()), Alpha=1, ConstraintWeights_0=1) {boneToModify {boneName=belly_l}, constraintSetup [{transformType=Translation, targetBone {boneName=foot_l}}], constraintWeights [1]}
   n2 = "Constraint - belly_r (T)"(ComponentPose=@n1, Alpha=1, ConstraintWeights_0=1) {boneToModify {boneName=belly_r}, constraintSetup [{transformType=Translation, targetBone {boneName=foot_r}}], constraintWeights [1]}
   Output Pose(Result=Component To Local(ComponentPose=@n2))
  ]}
 EventGraph {links [], nodes []}
]
variables [PoseAsset {type=PoseAsset, default=/CosmeticCompanions/Automation_Template/companion_pose_asset_template}]
